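# Material function for the CatSpace backpack cosmetic that generates looping, animated particle UVs and a matching fade value.
# A fractional time ramp, driven by speed and offset inputs, scales UVs about their center, with a static switch reversing the direction.
# The scaled UVs pass through a Rotator driven by an overridable rotation time to form the UV output.
# The Fading output fades in and back out over each cycle, hiding the loop reset.
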MaterialFunction /BRCosmetics/Accessories/FORT_Backpacks/Backpack_F_MED_CatSpace/Materials/MaterialFunctions/MF_CatSpace_Backpack_ParticleAnimation
graph {nodes [
  Constant() {r=1}
  n1 = Time()
  n2 = FunctionInput() {bUsePreviewValueAsDefault=true, inputName=ParticleSpeed, inputType=FunctionInput_Scalar, previewValue=<1 0 0 1>}
  n3 = Frac(None=Add(
    A=Multiply(A=@n1, B=@n2) {constB=0.5}
    B=FunctionInput() {bUsePreviewValueAsDefault=true, inputName=AnimationOffset, inputType=FunctionInput_Scalar, sortPriority=5}
   ) {constB=0})
  n4 = StaticSwitch(
   True=LinearInterpolate(Alpha=@n3) {constA=2}
   False=Add(A=@n3)
   Value=FunctionInput(Preview=StaticBool() {value=true}) {inputName=InvertParticleDirection, inputType=FunctionInput_StaticBool}
  )
  n5 = Time()
  FunctionOutput(None=Rotator(
    Coordinate=MaterialFunctionCall(Texture Scale=@n4) {materialFunction=&ScaleUVsByCenter}.UVs
    Time=FunctionInput(Preview=@n5) {bUsePreviewValueAsDefault=true, inputName=RotationTime, inputType=FunctionInput_Scalar, sortPriority=10}
   )) {outputName=UV}
  n6 = Multiply(A=@n3) {constB=2}
  n7 = Saturate(None=Subtract(A=@n6))
  n8 = LinearInterpolate(Alpha=Saturate(None=@n6))
  FunctionOutput(None=If(A=@n6, "A > B"=LinearInterpolate(Alpha=@n7) {constA=1, constB=0}, "A == B"=@n8, "A < B"=@n8) {constB=1}) {outputName=Fading}
 ]}
