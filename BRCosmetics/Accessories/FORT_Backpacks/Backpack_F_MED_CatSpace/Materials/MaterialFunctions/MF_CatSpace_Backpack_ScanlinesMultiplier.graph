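# Material function for the CatSpace backpack cosmetic that layers animated scanlines onto the emissive and opacity of incoming material attributes.
# A nested ScanlinesBase function supplies the pattern from rotated UVs, tiling, panner speed and PawAlpha.
# With the static IsPaw switch on, the pattern blends boosted emissive toward PawTexture and, multiplied by PawAlpha, blends overall into paw opacity.
# Otherwise it multiplies emissive, optionally limited by an FX mask channel, and leaves opacity untouched.
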MaterialFunction /BRCosmetics/Accessories/FORT_Backpacks/Backpack_F_MED_CatSpace/Materials/MaterialFunctions/MF_CatSpace_Backpack_ScanlinesMultiplier
graph {nodes [
  n1 = GetMaterialAttributes(None=FunctionInput() {inputType=FunctionInput_MaterialAttributes}) {attributeGetTypes [#1, #2]}
  n2 = ScalarParameter() {defaultValue=1, group=CatSpace_Scanlines, parameterName=PawScanline_ColorMultiplier}
  n3 = FunctionInput() {inputName=PawAlpha}
  n4 = MaterialFunctionCall!(
   ScanLine_Tiling_1
   Scanline_PannerSpeed_1
   ScanLine_Tiling
   Scanline_PannerSpeed
   PawAlpha=@n3
   UVs=Rotator(Coordinate=TextureCoordinate(), Time=ScalarParameter() {group=CatSpace_Scanlines, parameterName=ScanlinesRotation})
  ) {materialFunction=&/MF_CatSpace_Backpack_ScanlinesBase}
  n5 = LinearInterpolate(
   A=LinearInterpolate(A=@n1.EmissiveColor, B=Multiply(A=@n2, B=@n1.EmissiveColor), Alpha=@n3)
   B=FunctionInput() {bUsePreviewValueAsDefault=true, inputName=PawTexture}
   Alpha=@n4.Result
  )
  n6 = Multiply(A=@n1.EmissiveColor, B=@n4.Result)
  n7 = FunctionInput() {bUsePreviewValueAsDefault=true, inputName=FX mask}
  n8 = LinearInterpolate(
   A=@n6
   B=LinearInterpolate(A=@n1.EmissiveColor, B=@n6, Alpha=ChannelMaskParameter(None=@n7) {group=Scanline, parameterName=Mask Channel}.R)
   Alpha=ScalarParameter() {group=CatSpace_Scanlines, parameterName=Mask Scanlines}
  )
  n9 = FunctionInput() {bUsePreviewValueAsDefault=true, inputName=IsPaw, inputType=FunctionInput_StaticBool}
  n10 = LinearInterpolate(
   A=ScalarParameter() {defaultValue=0.75, group=CatSpace, parameterName=OverallOpacity, sortPriority=51}
   B=ScalarParameter() {defaultValue=0.75, group=CatSpace, parameterName=PawOpacity, sortPriority=52}
   Alpha=Multiply(A=@n3, B=@n4.Result)
  )
  FunctionOutput(None=SetMaterialAttributes!(
    MaterialAttributes=@n1.MaterialAttributes
    EmissiveColor=StaticSwitch(True=@n5, False=@n8, Value=@n9)
    Opacity=StaticSwitch(True=@n10, False=@n1.Opacity, Value=@n9)
   ) {attributeSetTypes [#1, #2]})
 ]}
