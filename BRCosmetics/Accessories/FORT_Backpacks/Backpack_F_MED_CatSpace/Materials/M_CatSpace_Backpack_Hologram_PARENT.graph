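# Translucent unlit parent material for the CatSpace backpack hologram, showing an animated paw and particles over a two-color vertical gradient.
# The paw texture intermittently glitches and wiggles on time-driven sine pulses.
# Particles combine two rotating layers, each sampled at two offset phases, with a flipbook.
# A scanline function processes base and paw passes, then emissive is divided by eye adaptation and scaled by time of day.
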Material /BRCosmetics/Accessories/FORT_Backpacks/Backpack_F_MED_CatSpace/Materials/M_CatSpace_Backpack_Hologram_PARENT
graph {
 properties {
  bUseMaterialAttributes=true
  bUsedWithMeshDeformer=true
  bUsedWithSkeletalMesh=true
  blendMode=BLEND_Translucent
  pixelDepthOffsetMode=PDOM_Legacy
  shadingModel=MSM_Unlit
  usedShadingModels=Unlit
 }
 nodes [
  Constant() {r=1}
  n1 = TextureCoordinate()
  n2 = Rotator(
   Coordinate=MaterialFunctionCall(UVs=@n1, Texture Scale=ScalarParameter() {defaultValue=1, group=CatSpace, parameterName=PawTextureScale}) {materialFunction=&ScaleUVsByCenter}.UVs
   Time=ScalarParameter() {group=CatSpace, parameterName=PawTextureRotation}
  )
  n3 = MaterialFunctionCall!(UV=@n2) {materialFunction=/Game/Weapons/WeaponSkins/MaterialFunction/MF_GlitchFunction}
  n4 = Time()
  n5 = Multiply(A=Sine(None=Multiply(A=@n4) {constB=10}), B=Constant() {desc=Wiggle Intensity, r=0.0085}) {constB=0.007}
  n6 = Time()
  n7 = Frac(None=Multiply(A=@n6) {constB=0.5})
  n8 = Saturate(None=ComponentMask(None=NamedRerouteUsage()) {r=true})
  n9 = OneMinus(None=@n8)
  n10 = Constant() {r=0.37}
  n11 = Multiply(
   A=@n6
   B=If(A=If(A=@n7, "A > B"=@n8, "A == B"=@n8, "A < B"=@n9) {constB=0.5}, "A > B"=@n10, "A == B"=@n10, "A < B"=Constant() {r=0.4}) {constB=1}
  ) {constB=0.6}
  n12 = Add(A=Multiply(A=Sine(None=@n11)) {constB=0.5}) {constB=0.5}
  n13 = TextureSampleParameter2D!(
   UVs=LinearInterpolate(A=@n2, B=Add(A=@n3.Glitched UV, B=@n5), Alpha=SmoothStep(Value=@n12) {constMax=0.972, constMin=0.967})
   "DDX(UVs)"=DDX(Value=@n1)
   "DDY(UVs)"=DDY(Value=@n1)
   Apply View MipBias
  ) {
   group=CatSpace
   mipValueMode=TMVM_Derivative
   parameterName=PawTexture
   samplerSource=SSM_Clamp_WorldGroupSettings
   samplerType=SAMPLERTYPE_Masks
   texture=/BRCosmetics/Accessories/FORT_Backpacks/Backpack_F_MED_CatSpace/Textures/T_CatSpace_Backpack_Paw
  }
  n14 = Add(A=NamedRerouteUsage(), B=@n13.RGB)
  NamedRerouteDeclaration(None=@n14) {name=PawAlpha, nodeColor=<rgba 0.9783 1 0.9366 1>}
  NamedRerouteDeclaration(None=Multiply(A=@n14, B=ScalarParameter() {defaultValue=1, group=CatSpace, parameterName=PawColorIntensity})) {name=PawTexture, nodeColor=<rgba 1 0.7266 0 1>}
  n15 = StaticBool()
  n16 = ScalarParameter() {defaultValue=0.5, group=CatSpace, parameterName=Particles01_Speed, sortPriority=10}
  n17 = Multiply(A=Time(), B=ScalarParameter() {defaultValue=0.8, group=CatSpace, parameterName=Particles01_RotationSpeed, sortPriority=10}) {constB=0.2}
  n18 = MaterialFunctionCall!(InvertParticleDirection=@n15, ParticleSpeed=@n16, AnimationOffset, RotationTime=@n17) {materialFunction=/BRCosmetics/Accessories/FORT_Backpacks/Backpack_F_MED_CatSpace/Materials/MaterialFunctions/MF_CatSpace_Backpack_ParticleAnimation}
  n19 = TextureSample(UVs=@n18.UV) {
   samplerSource=SSM_Clamp_WorldGroupSettings
   samplerType=SAMPLERTYPE_Masks
   texture=/BRCosmetics/Accessories/FORT_Backpacks/Backpack_F_MED_CatSpace/Textures/T_CatSpace_Backpack_PawParticles
  }
  n20 = MaterialFunctionCall(InvertParticleDirection=@n15, ParticleSpeed=@n16, AnimationOffset=Constant() {r=0.5}, RotationTime=@n17) {materialFunction=&MF_CatSpace_Backpack_ParticleAnimation}
  n21 = TextureSample(UVs=@n20.UV) {
   samplerSource=SSM_Clamp_WorldGroupSettings
   samplerType=SAMPLERTYPE_Masks
   texture=&T_CatSpace_Backpack_PawParticles
  }
  n22 = ScalarParameter() {defaultValue=0.42, group=CatSpace, parameterName=Particles02_Speed, sortPriority=10}
  n23 = Multiply(A=Time(), B=ScalarParameter() {defaultValue=0.2, group=CatSpace, parameterName=Particles02_RotationSpeed, sortPriority=10}) {constB=0.2}
  n24 = MaterialFunctionCall(InvertParticleDirection=@n15, ParticleSpeed=@n22, RotationTime=@n23) {materialFunction=&MF_CatSpace_Backpack_ParticleAnimation}
  n25 = TextureSample(UVs=@n24.UV) {
   samplerType=SAMPLERTYPE_Masks
   texture=&T_CatSpace_Backpack_PawParticles
  }
  n26 = MaterialFunctionCall(InvertParticleDirection=@n15, ParticleSpeed=@n22, AnimationOffset=Constant() {r=0.5}, RotationTime=@n23) {materialFunction=&MF_CatSpace_Backpack_ParticleAnimation}
  n27 = TextureSample(UVs=@n26.UV) {
   samplerType=SAMPLERTYPE_Masks
   texture=&T_CatSpace_Backpack_PawParticles
  }
  n28 = Add(
   A=Add(A=Multiply(A=@n18.Fading, B=@n19.R), B=Multiply(A=@n20.Fading, B=@n21.R))
   B=Add(A=Multiply(A=@n24.Fading, B=@n25.G), B=Multiply(A=@n26.Fading, B=@n27.G))
  )
  n29 = Multiply(A=Time()) {constB=0.9}
  n30 = Constant() {r=6}
  n31 = SmoothStep(Value=TextureSample(UVs=MaterialFunctionCall("Animation  Phase (0-1)"=@n29, Number of Rows=@n30, Number of Columns=@n30) {materialFunction=&FlipBook}.UVs) {texture=/BRCosmetics/Accessories/FORT_Backpacks/Backpack_F_MED_CatSpace/Textures/T_CatSpace_FlipbookParticles}.R) {constMax=0.2}
  NamedRerouteDeclaration(None=Add(A=Multiply(A=@n28) {constB=3}, B=Multiply(A=@n31) {constB=2})) {name=Particles, nodeColor=<rgba 1 0.2578 0 1>}
  n32 = VertexInterpolator(VS=PreSkinnedPosition())
  Reroute(None=@n32.PS)
  n33 = NamedRerouteUsage()
  n34 = VectorParameter() {group=CatSpace, parameterName=HologramColor_02, sortPriority=1, defaultValue=<rgba 0.3915 0.0425 1 1>}
  n35 = VectorParameter() {group=CatSpace, parameterName=HologramColor_01, sortPriority=1, defaultValue=<rgba 1 0.0638 0.6033 1>}
  n36 = Add(
   A=NamedRerouteDeclaration(None=@n32.PS) {name=LocalPositionPreSkinned, nodeColor=<rgba 1 0 0.9609 1>}
   B=Constant3Vector() {constant=<rgba 0 0 -2.5 0>}
  )
  n37 = Saturate(None=ComponentMask(None=Multiply(A=@n36) {constB=0.15}) {b=true})
  n38 = MaterialFunctionCall!(
   FX mask
   PawAlpha=Constant()
   PawTexture=@n33
   IsPaw=StaticBool()
   In=SetMaterialAttributes!(
    MaterialAttributes
    Emissive Color=LinearInterpolate(A=@n34.RGB, B=@n35.RGB, Alpha=@n37)
    Opacity=ScalarParameter() {defaultValue=0.75, group=CatSpace, parameterName=OverallOpacity, sortPriority=51}
   ) {attributeSetTypes [#1, #2]}
  ) {materialFunction=/BRCosmetics/Accessories/FORT_Backpacks/Backpack_F_MED_CatSpace/Materials/MaterialFunctions/MF_CatSpace_Backpack_ScanlinesMultiplier}
  n39 = MaterialFunctionCall(PawAlpha=NamedRerouteUsage(), PawTexture=@n33, IsPaw=StaticBool() {value=true}, In=@n38.Result) {materialFunction=&MF_CatSpace_Backpack_ScanlinesMultiplier}
  n40 = GetMaterialAttributes(None=@n39.Result) {attributeGetTypes [#1]}
  n41 = ComponentMask(None=@n40.Emissive Color) {b=true, g=true, r=true}
  n42 = EyeAdaptation()
  n43 = Constant() {r=1}
  n44 = Constant() {r=0.7}
  SetMaterialAttributes!(
   MaterialAttributes=@n40.MaterialAttributes
   Emissive Color=Multiply(
    A=Divide(A=@n41, B=@n42.EyeAdaptation)
    B=MaterialFunctionCall(Value_Morning=@n43, Value_Day=@n43, Value_Evening=@n43, Value_Night=@n44) {materialFunction=&Scalar_Time_of_Day}.Result
   )
  ) {attributeSetTypes [#1]}
 ]
}
