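# Animation Blueprint for the CatSpace backpack, posing its wings to match player movement.
# A wingPosition state machine, with a nested skydiving machine, selects pose-asset poses, most layered with additive ground or air lean blendspaces driven by movement.
# AnimDynamics then simulates disk and wing bones, with disk strength raised during skydiving.
# Finally, bone transforms add a sine-driven idle float to each wing, computed every update from real time.
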AnimBlueprint /BRCosmetics/Accessories/FORT_Backpacks/Backpack_F_MED_CatSpace/Meshes/F_MED_CatSpace_Pack_AnimBP
target_skeleton=/Game/Accessories/FORT_Backpacks/Backpack_F_MED_CatSpace/Meshes/F_MED_CatSpace_Pack_Skeleton
graphs [
 AnimGraph {nodes [
   AnimGraphNode_StateMachine_0 = wingPosition() {is=title AnimGraphNode_StateMachine}
   n1 = Get DynamicsAlpha() -> (DynamicsAlpha:double)
   n2 = Anim Dynamics - Bone: dyn_disk_l_jnt(ComponentPose=Local To Component(LocalPose=@AnimGraphNode_StateMachine_0), Alpha=@n1, ExternalForce="0,0,0") {
    alpha=0.5
    angularDampingOverride=0.9999
    angularSpringConstant=8000
    bAngularSpring=true
    bGravityOverrideInSimSpace=true
    bOverrideAngularDamping=true
    bUseGravityOverride=true
    linearSpringConstant=10
    simulationSpace=RootRelative
    alphaBoolBlend {blendOption=Linear}
    boundBone {boneName=dyn_disk_l_jnt}
    componentAppliedLinearAccClamp=<1000 1000 1000>
    componentLinearAccScale=<1 1 1>
    componentLinearVelScale=<1 1 1>
    gravityOverride {z=-980}
    physicsBodyDefinitions [{
      sphereCollisionRadius=0
      boundBone {boneName=dyn_disk_l_jnt}
      boxExtents=<12 18 18>
      constraintSetup {angularTargetAxis=AxisZ, angularLimitsMax=<0 15 -2>, angularLimitsMin=<0 -15 -2>, angularTarget=<0 0 1>}
      localJointOffset=<0.5 0 0>
     }]
   }
   n3 = Anim Dynamics - Bone: dyn_wing_spin_l_jnt(ComponentPose=@n2, Alpha=0.4, ExternalForce="0,0,0") {
    alpha=0.5
    angularDampingOverride=0.9999
    angularSpringConstant=8000
    bGravityOverrideInSimSpace=true
    bLinearSpring=true
    bOverrideLinearDamping=true
    bUseGravityOverride=true
    linearDampingOverride=0.999
    linearSpringConstant=40
    simulationSpace=RootRelative
    alphaBoolBlend {blendInTime=0.1, blendOption=Linear, blendOutTime=0.1}
    alphaScaleBiasClamp {scale=0.4}
    boundBone {boneName=dyn_wing_spin_l_jnt}
    componentAppliedLinearAccClamp=<100 100 100>
    componentLinearAccScale=<0.1 0.1 0.1>
    componentLinearVelScale=<0.03 0.03 0.03>
    gravityOverride {z=-980}
    physicsBodyDefinitions [{
      sphereCollisionRadius=0
      boundBone {boneName=dyn_wing_spin_l_jnt}
      boxExtents=<100 15 15>
      constraintSetup {
       angularTargetAxis=AxisZ
       angularLimitsMax=<0 0 0>
       angularLimitsMin=<0 0 0>
       angularTarget=<0 0 1>
       linearAxesMax=<8 0 0>
       linearAxesMin=<-0.2 0 0>
      }
      localJointOffset=<0.1 0 0>
     }]
   }
   n4 = Anim Dynamics - Bone: dyn_disk_r_jnt(ComponentPose=@n3, Alpha=@n1, ExternalForce="0,0,0") {
    alpha=0.5
    angularDampingOverride=0.9999
    angularSpringConstant=8000
    bAngularSpring=true
    bGravityOverrideInSimSpace=true
    bOverrideAngularDamping=true
    bUseGravityOverride=true
    simulationSpace=RootRelative
    alphaBoolBlend {blendOption=Linear}
    boundBone {boneName=dyn_disk_r_jnt}
    componentAppliedLinearAccClamp=<1000 1000 1000>
    componentLinearAccScale=<1 1 1>
    componentLinearVelScale=<1 1 1>
    gravityOverride {z=-980}
    physicsBodyDefinitions [{
      sphereCollisionRadius=0
      boundBone {boneName=dyn_disk_r_jnt}
      boxExtents=<12 18 18>
      constraintSetup {angularTargetAxis=AxisZ, angularLimitsMax=<0 15 2>, angularLimitsMin=<0 -15 -2>, angularTarget=<0 0 1>}
      localJointOffset=<0.5 0 0>
     }]
   }
   n5 = Anim Dynamics - Bone: dyn_wing_spin_r_jnt(ComponentPose=@n4, Alpha=0.4, ExternalForce="0,0,0") {
    alpha=0.5
    angularDampingOverride=0.9999
    angularSpringConstant=8000
    bGravityOverrideInSimSpace=true
    bLinearSpring=true
    bOverrideLinearDamping=true
    bUseGravityOverride=true
    linearDampingOverride=0.999
    linearSpringConstant=40
    simulationSpace=RootRelative
    alphaBoolBlend {blendOption=Linear}
    alphaScaleBiasClamp {scale=0.4}
    boundBone {boneName=dyn_wing_spin_r_jnt}
    componentAppliedLinearAccClamp=<100 100 100>
    componentLinearAccScale=<0.1 0.1 0.1>
    componentLinearVelScale=<0.03 0.03 0.03>
    gravityOverride {z=-980}
    physicsBodyDefinitions [{
      sphereCollisionRadius=0
      boundBone {boneName=dyn_wing_spin_r_jnt}
      boxExtents=<100 15 15>
      constraintSetup {
       angularTargetAxis=AxisZ
       angularLimitsMax=<0 0 0>
       angularLimitsMin=<0 0 0>
       angularTarget=<0 0 1>
       linearAxesMax=<0.2 0 0>
       linearAxesMin=<-8 0 0>
      }
      localJointOffset=<0.1 0 0>
     }]
   }
   n6 = "Transform (Modify) Bone - Bone: dyn_wing_spin_l_jnt"(ComponentPose=@n5, Alpha=0.75, Translation=Get dyn_l_pack_Float() -> (dyn_l_pack_Float:Vector), Rotation="-5,0,0", Scale="1,1,1") {
    rotationMode=BMM_Additive
    rotationSpace=BCS_BoneSpace
    translationMode=BMM_Additive
    translationSpace=BCS_BoneSpace
    alphaBoolBlend {blendOption=Linear}
    alphaScaleBiasClamp {bInterpResult=true, interpSpeedDecreasing=10000, interpSpeedIncreasing=10000}
    boneToModify {boneName=dyn_wing_spin_l_jnt}
    rotation {pitch=-5}
   }
   n7 = "Transform (Modify) Bone - Bone: dyn_wing_spin_r_jnt"(ComponentPose=@n6, Alpha=0.75, Translation=Get dyn_r_pack_Float() -> (dyn_r_pack_Float:Vector), Rotation="0,0,0", Scale="1,1,1") {
    rotationSpace=BCS_BoneSpace
    translationMode=BMM_Additive
    translationSpace=BCS_BoneSpace
    alphaBoolBlend {blendOption=Linear}
    boneToModify {boneName=dyn_wing_spin_r_jnt}
   }
   Output Pose(Result=Component To Local(ComponentPose=@n7)) {layerGroup=None}
  ]}
 wingPosition {
  entry=Base_1
  path=AnimGraph.AnimGraphNode_StateMachine_0.wingPosition
  ids=id
  nodes [
   AnimStateNode_0 = Base_1() {is=title AnimStateNode}
   AnimStateNode_1 = Jumping_1() {is=title AnimStateNode}
   AnimStateNode_2 = Falling_1() {is=title AnimStateNode}
   AnimStateNode_3 = Targeting_1() {is=title AnimStateNode}
   AnimStateNode_4 = Crouching_1() {is=title AnimStateNode}
   AnimStateNode_5 = SkydiveMode_1() {is=title AnimStateNode}
   AnimStateNode_9 = Tactical Sprint_1() {is=title AnimStateNode}
   AnimStateTransitionNode_0 = AnimStateTransitionNode() {
    blendMode=Custom
    customBlendCurve=/BRCosmetics/Accessories/FORT_Backpacks/Backpack_F_MED_CatSpace/Animation/JumpTransitionBlendCurve
    transitionEnd {guid=#1}
    transitionInterrupt {guid=#2}
    transitionStart {guid=#3}
   }
   AnimStateTransitionNode_1 = AnimStateTransitionNode() {
    blendMode=Custom
    crossfadeDuration=0.5
    customBlendCurve=/BRCosmetics/Accessories/FORT_Backpacks/Backpack_F_MED_CatSpace/Animation/JumpTransitionBlendOutCurve
   }
   AnimStateTransitionNode_10 = AnimStateTransitionNode()
   AnimStateTransitionNode_11 = AnimStateTransitionNode()
   AnimStateTransitionNode_12 = AnimStateTransitionNode()
   AnimStateTransitionNode_13 = AnimStateTransitionNode()
   AnimStateTransitionNode_14 = AnimStateTransitionNode()
   AnimStateTransitionNode_15 = AnimStateTransitionNode()
   AnimStateTransitionNode_2 = AnimStateTransitionNode()
   AnimStateTransitionNode_3 = AnimStateTransitionNode() {blendMode=Custom, customBlendCurve=&JumpTransitionBlendOutCurve}
   AnimStateTransitionNode_4 = AnimStateTransitionNode() {blendMode=Custom, customBlendCurve=&JumpTransitionBlendCurve}
   AnimStateTransitionNode_5 = AnimStateTransitionNode()
   AnimStateTransitionNode_6 = AnimStateTransitionNode() {crossfadeDuration=0.25}
   AnimStateTransitionNode_7 = AnimStateTransitionNode() {crossfadeDuration=0.25}
   AnimStateTransitionNode_8 = AnimStateTransitionNode()
   AnimStateTransitionNode_9 = AnimStateTransitionNode()
  ]
  links [
   {
    dst=Crouching_1
    rule=AnimGraph.AnimGraphNode_StateMachine_0.wingPosition.AnimStateTransitionNode_12.AnimationTransitionGraph_0
    src=Base_1
    via=AnimStateTransitionNode_12
   }
   {
    dst=Falling_1
    rule=AnimGraph.AnimGraphNode_StateMachine_0.wingPosition.AnimStateTransitionNode_4.AnimationTransitionGraph_0
    src=Base_1
    via=AnimStateTransitionNode_4
   }
   {
    dst=Jumping_1
    rule=AnimGraph.AnimGraphNode_StateMachine_0.wingPosition.AnimStateTransitionNode_0.AnimationTransitionGraph_0
    src=Base_1
    via=AnimStateTransitionNode_0
   }
   {
    dst=SkydiveMode_1
    rule=AnimGraph.AnimGraphNode_StateMachine_0.wingPosition.AnimStateTransitionNode_15.AnimationTransitionGraph_0
    src=Base_1
    via=AnimStateTransitionNode_15
   }
   {
    dst=Tactical Sprint_1
    rule=AnimGraph.AnimGraphNode_StateMachine_0.wingPosition.AnimStateTransitionNode_8.AnimationTransitionGraph_0
    src=Base_1
    via=AnimStateTransitionNode_8
   }
   {
    dst=Targeting_1
    rule=AnimGraph.AnimGraphNode_StateMachine_0.wingPosition.AnimStateTransitionNode_6.AnimationTransitionGraph_0
    src=Base_1
    via=AnimStateTransitionNode_6
   }
   {
    dst=Base_1
    rule=AnimGraph.AnimGraphNode_StateMachine_0.wingPosition.AnimStateTransitionNode_13.AnimationTransitionGraph_0
    src=Crouching_1
    via=AnimStateTransitionNode_13
   }
   {
    dst=Base_1
    rule=AnimGraph.AnimGraphNode_StateMachine_0.wingPosition.AnimStateTransitionNode_3.AnimationTransitionGraph_0
    src=Falling_1
    via=AnimStateTransitionNode_3
   }
   {
    dst=Base_1
    rule=AnimGraph.AnimGraphNode_StateMachine_0.wingPosition.AnimStateTransitionNode_1.AnimationTransitionGraph_0
    src=Jumping_1
    via=AnimStateTransitionNode_1
   }
   {
    dst=Falling_1
    rule=AnimGraph.AnimGraphNode_StateMachine_0.wingPosition.AnimStateTransitionNode_2.AnimationTransitionGraph_0
    src=Jumping_1
    via=AnimStateTransitionNode_2
   }
   {
    dst=SkydiveMode_1
    rule=AnimGraph.AnimGraphNode_StateMachine_0.wingPosition.AnimStateTransitionNode_5.AnimationTransitionGraph_0
    src=Jumping_1
    via=AnimStateTransitionNode_5
   }
   {
    dst=Tactical Sprint_1
    rule=AnimGraph.AnimGraphNode_StateMachine_0.wingPosition.AnimStateTransitionNode_14.AnimationTransitionGraph_0
    src=Jumping_1
    via=AnimStateTransitionNode_14
   }
   {
    dst=Base_1
    rule=AnimGraph.AnimGraphNode_StateMachine_0.wingPosition.AnimStateTransitionNode_11.AnimationTransitionGraph_0
    src=SkydiveMode_1
    via=AnimStateTransitionNode_11
   }
   {
    dst=Base_1
    rule=AnimGraph.AnimGraphNode_StateMachine_0.wingPosition.AnimStateTransitionNode_9.AnimationTransitionGraph_0
    src=Tactical Sprint_1
    via=AnimStateTransitionNode_9
   }
   {
    dst=Jumping_1
    rule=AnimGraph.AnimGraphNode_StateMachine_0.wingPosition.AnimStateTransitionNode_10.AnimationTransitionGraph_0
    src=Tactical Sprint_1
    via=AnimStateTransitionNode_10
   }
   {
    dst=Base_1
    rule=AnimGraph.AnimGraphNode_StateMachine_0.wingPosition.AnimStateTransitionNode_7.AnimationTransitionGraph_0
    src=Targeting_1
    via=AnimStateTransitionNode_7
   }
  ]
 }
 Base_1 {
  path=AnimGraph.AnimGraphNode_StateMachine_0.wingPosition.AnimStateNode_0.Base_1
  nodes [
   n1 = Get PlayerVector_FwdBwd() -> (PlayerVector_FwdBwd:double)
   n2 = Get PlayerVector_LR() -> (PlayerVector_LR:double)
   Output Animation Pose(Result=Apply Additive(
     Base=Wing_Movement_Poses_PoseAsset : Base!(PoseWeight:float=1) -> (Pose:Pose Link Structure) {
      is=title AnimGraphNode_PoseByName
      poseAsset=/BRCosmetics/Accessories/FORT_Backpacks/Backpack_F_MED_CatSpace/Animation/Wing_Movement_Poses_PoseAsset
      poseName=Base
     }
     Additive=Blendspace Player 'BS_Wing_Leans_Ground'(X=@n2, Y=@n1) {blendSpace=/BRCosmetics/Accessories/FORT_Backpacks/Backpack_F_MED_CatSpace/Animation/BS_Wing_Leans_Ground}
    ) {alphaBoolBlend {blendOption=Linear}})
  ]
 }
 Jumping_1 {
  path=AnimGraph.AnimGraphNode_StateMachine_0.wingPosition.AnimStateNode_1.Jumping_1
  nodes [
   n1 = Get PlayerVector_FwdBwd()
   n2 = Get PlayerVector_LR()
   Output Animation Pose(Result=Apply Additive(
     Base=Wing_Movement_Poses_PoseAsset : Jump!(PoseWeight:float=1) -> (Pose:Pose Link Structure) {
      is=title AnimGraphNode_PoseByName
      poseAsset=&Wing_Movement_Poses_PoseAsset
      poseName=Jump
     }
     Additive=Blendspace Player 'BS_Wing_Leans_Ground'(X=@n2, Y=@n1) {blendSpace=&BS_Wing_Leans_Ground}
    ) {alphaBoolBlend {blendOption=Linear}})
  ]
 }
 Falling_1 {
  path=AnimGraph.AnimGraphNode_StateMachine_0.wingPosition.AnimStateNode_2.Falling_1
  nodes [
   n1 = Get PlayerVector_FwdBwd()
   n2 = Get PlayerVector_LR()
   Output Animation Pose(Result=Apply Additive(
     Base=Wing_Movement_Poses_PoseAsset : Jump(PoseWeight=1) {poseAsset=&Wing_Movement_Poses_PoseAsset, poseName=Jump}
     Additive=Blendspace Player 'BS_Wing_Leans_Ground'(X=@n2, Y=@n1) {blendSpace=&BS_Wing_Leans_Ground}
    ) {alphaBoolBlend {blendOption=Linear}})
  ]
 }
 Targeting_1 {
  path=AnimGraph.AnimGraphNode_StateMachine_0.wingPosition.AnimStateNode_3.Targeting_1
  nodes [
   n1 = Wing_Movement_Poses_PoseAsset : Targeting!(PoseWeight:float=1) -> (Pose:Pose Link Structure) {
    is=title AnimGraphNode_PoseByName
    poseAsset=&Wing_Movement_Poses_PoseAsset
    poseName=Targeting
   }
   Output Animation Pose(Result=@n1)
  ]
 }
 Crouching_1 {
  path=AnimGraph.AnimGraphNode_StateMachine_0.wingPosition.AnimStateNode_4.Crouching_1
  nodes [Output Animation Pose(Result=Wing_Movement_Poses_PoseAsset : Crouch!(PoseWeight:float=1) -> (Pose:Pose Link Structure) {
     is=title AnimGraphNode_PoseByName
     poseAsset=&Wing_Movement_Poses_PoseAsset
     poseName=Crouch
    })]
 }
 SkydiveMode_1 {
  path=AnimGraph.AnimGraphNode_StateMachine_0.wingPosition.AnimStateNode_5.SkydiveMode_1
  nodes [
   AnimGraphNode_StateMachine_0 = Skydiving State() {is=title AnimGraphNode_StateMachine}
   Output Animation Pose(Result=@AnimGraphNode_StateMachine_0)
  ]
 }
 Skydiving State {
  entry=Parachuting
  path=AnimGraph.AnimGraphNode_StateMachine_0.wingPosition.AnimStateNode_5.SkydiveMode_1.AnimGraphNode_StateMachine_0.Skydiving State
  ids=id
  nodes [
   AnimStateNode_1 = Diving() {is=title AnimStateNode}
   AnimStateNode_2 = Floating() {is=title AnimStateNode}
   AnimStateNode_4 = Parachuting() {is=title AnimStateNode}
   AnimStateTransitionNode_1 = AnimStateTransitionNode() {crossfadeDuration=0.5}
   AnimStateTransitionNode_10 = AnimStateTransitionNode()
   AnimStateTransitionNode_12 = AnimStateTransitionNode()
   AnimStateTransitionNode_13 = AnimStateTransitionNode()
   AnimStateTransitionNode_3 = AnimStateTransitionNode()
   AnimStateTransitionNode_5 = AnimStateTransitionNode() {crossfadeDuration=0.5}
  ]
  links [
   {
    dst=Floating
    rule=AnimGraph.AnimGraphNode_StateMachine_0.wingPosition.AnimStateNode_5.SkydiveMode_1.AnimGraphNode_StateMachine_0.Skydiving State.AnimStateTransitionNode_1.AnimationTransitionGraph_0
    src=Diving
    via=AnimStateTransitionNode_1
   }
   {
    dst=Parachuting
    rule=AnimGraph.AnimGraphNode_StateMachine_0.wingPosition.AnimStateNode_5.SkydiveMode_1.AnimGraphNode_StateMachine_0.Skydiving State.AnimStateTransitionNode_10.AnimationTransitionGraph_1
    src=Diving
    via=AnimStateTransitionNode_10
   }
   {
    dst=Diving
    rule=AnimGraph.AnimGraphNode_StateMachine_0.wingPosition.AnimStateNode_5.SkydiveMode_1.AnimGraphNode_StateMachine_0.Skydiving State.AnimStateTransitionNode_5.AnimationTransitionGraph_0
    src=Floating
    via=AnimStateTransitionNode_5
   }
   {
    dst=Parachuting
    rule=AnimGraph.AnimGraphNode_StateMachine_0.wingPosition.AnimStateNode_5.SkydiveMode_1.AnimGraphNode_StateMachine_0.Skydiving State.AnimStateTransitionNode_3.AnimationTransitionGraph_1
    src=Floating
    via=AnimStateTransitionNode_3
   }
   {
    dst=Diving
    rule=AnimGraph.AnimGraphNode_StateMachine_0.wingPosition.AnimStateNode_5.SkydiveMode_1.AnimGraphNode_StateMachine_0.Skydiving State.AnimStateTransitionNode_13.AnimationTransitionGraph_1
    src=Parachuting
    via=AnimStateTransitionNode_13
   }
   {
    dst=Floating
    rule=AnimGraph.AnimGraphNode_StateMachine_0.wingPosition.AnimStateNode_5.SkydiveMode_1.AnimGraphNode_StateMachine_0.Skydiving State.AnimStateTransitionNode_12.AnimationTransitionGraph_1
    src=Parachuting
    via=AnimStateTransitionNode_12
   }
  ]
 }
 Diving {
  path=AnimGraph.AnimGraphNode_StateMachine_0.wingPosition.AnimStateNode_5.SkydiveMode_1.AnimGraphNode_StateMachine_0.Skydiving State.AnimStateNode_1.Diving
  nodes [
   n1 = Get PlayerVector_FwdBwd()
   n2 = Get PlayerVector_LR()
   Output Animation Pose(Result=Apply Additive(
     Base=Wing_Movement_Poses_PoseAsset : Dive!(PoseWeight:float=1) -> (Pose:Pose Link Structure) {
      is=title AnimGraphNode_PoseByName
      poseAsset=&Wing_Movement_Poses_PoseAsset
      poseName=Dive
     }
     Additive=Blendspace Player 'BS_Wing_Leans_Air'(X=@n2, Y=@n1) {blendSpace=/BRCosmetics/Accessories/FORT_Backpacks/Backpack_F_MED_CatSpace/Animation/BS_Wing_Leans_Air}
    ) {alphaBoolBlend {blendOption=Linear}})
  ]
 }
 Floating {
  path=AnimGraph.AnimGraphNode_StateMachine_0.wingPosition.AnimStateNode_5.SkydiveMode_1.AnimGraphNode_StateMachine_0.Skydiving State.AnimStateNode_2.Floating
  nodes [
   n1 = Get PlayerVector_FwdBwd()
   n2 = Get PlayerVector_LR()
   Output Animation Pose(Result=Apply Additive(
     Base=Wing_Movement_Poses_PoseAsset : Float!(PoseWeight:float=1) -> (Pose:Pose Link Structure) {
      is=title AnimGraphNode_PoseByName
      poseAsset=&Wing_Movement_Poses_PoseAsset
      poseName=Float
     }
     Additive=Blendspace Player 'BS_Wing_Leans_Air'(X=@n2, Y=@n1) {blendSpace=&BS_Wing_Leans_Air}
    ) {alphaBoolBlend {blendOption=Linear}})
  ]
 }
 Parachuting {
  path=AnimGraph.AnimGraphNode_StateMachine_0.wingPosition.AnimStateNode_5.SkydiveMode_1.AnimGraphNode_StateMachine_0.Skydiving State.AnimStateNode_4.Parachuting
  nodes [
   n1 = Get PlayerVector_FwdBwd()
   n2 = Get PlayerVector_LR()
   Output Animation Pose(Result=Apply Additive(
     Base=Wing_Movement_Poses_PoseAsset : Float(PoseWeight=1) {poseAsset=&Wing_Movement_Poses_PoseAsset, poseName=Float}
     Additive=Blendspace Player 'BS_Wing_Leans_Air'(X=@n2, Y=@n1) {blendSpace=&BS_Wing_Leans_Air}
    ) {alphaBoolBlend {blendOption=Linear}})
  ]
 }
 AnimationTransitionGraph_0 {
  path=AnimGraph.AnimGraphNode_StateMachine_0.wingPosition.AnimStateNode_5.SkydiveMode_1.AnimGraphNode_StateMachine_0.Skydiving State.AnimStateTransitionNode_1.AnimationTransitionGraph_0
  nodes [Result(bCanEnterTransition=NOT Boolean(A=Get bIsSkydiveDiveMode()))]
 }
 AnimationTransitionGraph_1 {
  path=AnimGraph.AnimGraphNode_StateMachine_0.wingPosition.AnimStateNode_5.SkydiveMode_1.AnimGraphNode_StateMachine_0.Skydiving State.AnimStateTransitionNode_10.AnimationTransitionGraph_1
  nodes [Result(bCanEnterTransition=Get bIsParachuteOpened())]
 }
 AnimationTransitionGraph_1 {
  path=AnimGraph.AnimGraphNode_StateMachine_0.wingPosition.AnimStateNode_5.SkydiveMode_1.AnimGraphNode_StateMachine_0.Skydiving State.AnimStateTransitionNode_12.AnimationTransitionGraph_1
  nodes [Result(bCanEnterTransition=NOT Boolean(A=Get bIsParachuteOpened()))]
 }
 AnimationTransitionGraph_1 {
  path=AnimGraph.AnimGraphNode_StateMachine_0.wingPosition.AnimStateNode_5.SkydiveMode_1.AnimGraphNode_StateMachine_0.Skydiving State.AnimStateTransitionNode_13.AnimationTransitionGraph_1
  nodes [Result(bCanEnterTransition=NOT Boolean(A=Get bIsParachuteOpened()))]
 }
 AnimationTransitionGraph_1 {
  path=AnimGraph.AnimGraphNode_StateMachine_0.wingPosition.AnimStateNode_5.SkydiveMode_1.AnimGraphNode_StateMachine_0.Skydiving State.AnimStateTransitionNode_3.AnimationTransitionGraph_1
  nodes [Result(bCanEnterTransition=Get bIsParachuteOpened())]
 }
 AnimationTransitionGraph_0 {
  path=AnimGraph.AnimGraphNode_StateMachine_0.wingPosition.AnimStateNode_5.SkydiveMode_1.AnimGraphNode_StateMachine_0.Skydiving State.AnimStateTransitionNode_5.AnimationTransitionGraph_0
  nodes [Result(bCanEnterTransition=Get bIsSkydiveDiveMode())]
 }
 Tactical Sprint_1 {
  path=AnimGraph.AnimGraphNode_StateMachine_0.wingPosition.AnimStateNode_9.Tactical Sprint_1
  nodes [Output Animation Pose(Result=Wing_Movement_Poses_PoseAsset : Sprint!(PoseWeight:float=1) -> (Pose:Pose Link Structure) {
     is=title AnimGraphNode_PoseByName
     poseAsset=&Wing_Movement_Poses_PoseAsset
     poseName=Sprint
    })]
 }
 AnimationTransitionGraph_0 {
  path=AnimGraph.AnimGraphNode_StateMachine_0.wingPosition.AnimStateTransitionNode_0.AnimationTransitionGraph_0
  nodes [Result(bCanEnterTransition=Get bIsJumping())]
 }
 AnimationTransitionGraph_0 {
  path=AnimGraph.AnimGraphNode_StateMachine_0.wingPosition.AnimStateTransitionNode_1.AnimationTransitionGraph_0
  nodes [
   Result(bCanEnterTransition=NOT Boolean(A=Get bIsJumping()))
   Get bShouldPredictLanding() -> (bShouldPredictLanding:bool)
   Get bLandingPredicted() -> (bLandingPredicted:bool)
   Get bDebugLandPrediction() -> (bDebugLandPrediction:bool)
   Get bIsFalling()
  ]
 }
 AnimationTransitionGraph_0 {
  path=AnimGraph.AnimGraphNode_StateMachine_0.wingPosition.AnimStateTransitionNode_10.AnimationTransitionGraph_0
  nodes [Result(bCanEnterTransition=Get bIsJumping())]
 }
 AnimationTransitionGraph_0 {
  path=AnimGraph.AnimGraphNode_StateMachine_0.wingPosition.AnimStateTransitionNode_11.AnimationTransitionGraph_0
  nodes [Result(bCanEnterTransition=NOT Boolean(A=Get bIsSkyDiving()))]
 }
 AnimationTransitionGraph_0 {
  path=AnimGraph.AnimGraphNode_StateMachine_0.wingPosition.AnimStateTransitionNode_12.AnimationTransitionGraph_0
  nodes [Result(bCanEnterTransition=Get bIsCrouching())]
 }
 AnimationTransitionGraph_0 {
  path=AnimGraph.AnimGraphNode_StateMachine_0.wingPosition.AnimStateTransitionNode_13.AnimationTransitionGraph_0
  nodes [Result(bCanEnterTransition=NOT Boolean(A=Get bIsCrouching()))]
 }
 AnimationTransitionGraph_0 {
  path=AnimGraph.AnimGraphNode_StateMachine_0.wingPosition.AnimStateTransitionNode_14.AnimationTransitionGraph_0
  nodes [Result(bCanEnterTransition=NOT Boolean(A=Get bIsJumping()))]
 }
 AnimationTransitionGraph_0 {
  path=AnimGraph.AnimGraphNode_StateMachine_0.wingPosition.AnimStateTransitionNode_15.AnimationTransitionGraph_0
  nodes [Result(bCanEnterTransition=Get bIsSkyDiving())]
 }
 AnimationTransitionGraph_0 {
  path=AnimGraph.AnimGraphNode_StateMachine_0.wingPosition.AnimStateTransitionNode_2.AnimationTransitionGraph_0
  nodes [Result(bCanEnterTransition=Get bIsFalling())]
 }
 AnimationTransitionGraph_0 {
  path=AnimGraph.AnimGraphNode_StateMachine_0.wingPosition.AnimStateTransitionNode_3.AnimationTransitionGraph_0
  nodes [Result(bCanEnterTransition=NOT Boolean(A=Get bIsFalling()))]
 }
 AnimationTransitionGraph_0 {
  path=AnimGraph.AnimGraphNode_StateMachine_0.wingPosition.AnimStateTransitionNode_4.AnimationTransitionGraph_0
  nodes [Result(bCanEnterTransition=Get bIsFalling())]
 }
 AnimationTransitionGraph_0 {
  path=AnimGraph.AnimGraphNode_StateMachine_0.wingPosition.AnimStateTransitionNode_5.AnimationTransitionGraph_0
  nodes [Result(bCanEnterTransition=Get bIsSkyDiving())]
 }
 AnimationTransitionGraph_0 {
  path=AnimGraph.AnimGraphNode_StateMachine_0.wingPosition.AnimStateTransitionNode_6.AnimationTransitionGraph_0
  nodes [Result(bCanEnterTransition=Get bIsTargeting())]
 }
 AnimationTransitionGraph_0 {
  path=AnimGraph.AnimGraphNode_StateMachine_0.wingPosition.AnimStateTransitionNode_7.AnimationTransitionGraph_0
  nodes [Result(bCanEnterTransition=NOT Boolean(A=Get bIsTargeting()))]
 }
 AnimationTransitionGraph_0 {
  path=AnimGraph.AnimGraphNode_StateMachine_0.wingPosition.AnimStateTransitionNode_8.AnimationTransitionGraph_0
  nodes [Result(bCanEnterTransition=Get bIsTacticalSprinting())]
 }
 AnimationTransitionGraph_0 {
  path=AnimGraph.AnimGraphNode_StateMachine_0.wingPosition.AnimStateTransitionNode_9.AnimationTransitionGraph_0
  nodes [Result(bCanEnterTransition=NOT Boolean(A=Get bIsTacticalSprinting()))]
 }
 EventGraph {nodes [
   Event BlueprintUpdateAnimation()
   > n1 = Sequence!(execute) -> (then_0, then_1, then_2:Exec, then_3:Exec)
   > n2 = PlayerMovement(AnimInstance=Self-Reference(), XDotProductCoefficient=1, YDotProductCoefficient=1, ZDotProductCoefficient=1)
   > Set PlayerVector_LR(PlayerVector_LR:double=@n2) -> (Output_Get:double)
   > Set PlayerVector_FwdBwd(PlayerVector_FwdBwd:double=@n2.YMovement) -> (Output_Get:double)
   n3 = GetRealTimeSeconds()
   n4 = "Sin (Radians)"(A=float * float(A=1.36, B=@n3))
   n5 = "Sin (Radians)"(A=float * float(A=1.2321, B=@n3))
   Set dyn_l_pack_Float(execute=@n1.then_1, dyn_l_pack_Float:Vector=MakeVector(X=float * float(A=@n4, B=0.94), Y=float * float(A=@n5, B=0.1))) -> (Output_Get:Vector)
   n6 = GetRealTimeSeconds()
   n7 = "Sin (Radians)"(A=float * float(A=1.1, B=@n6))
   n8 = "Sin (Radians)"(A=float * float(A=0.99, B=@n6))
   Set dyn_r_pack_Float(execute=@n1.then_2, dyn_r_pack_Float:Vector=MakeVector(X=float * float(A=@n8, B=0.85), Y=float * float(A=@n7, B=0.1))) -> (Output_Get:Vector)
   Set bl_isnt_skydiving(execute=@n1.then_3, bl_isnt_skydiving:bool="Not Equal (Boolean)"(A=Get bIsSkyDiving(), B=true) -> (ReturnValue=~)) -> (Output_Get:bool)
   Event OnBeginSkydiving()
   > Set DynamicsAlpha(DynamicsAlpha:double=1) -> (Output_Get:double)
   Event OnEndSkydiving()
   > Set DynamicsAlpha(DynamicsAlpha=0.5)
  ]}
]
variables [
 bl_isnt_skydiving {type=boolean}
 dyn_l_pack_Float {type=Vector, default=<-0.1172 -0.0733 -0.209>}
 dyn_r_pack_Float {type=Vector, default=<-0.2783 -0.0961 0.3545>}
 DynamicsAlpha {type=number, default=0.5}
 PlayerVector_FwdBwd {type=number}
 PlayerVector_LR {type=number}
]
